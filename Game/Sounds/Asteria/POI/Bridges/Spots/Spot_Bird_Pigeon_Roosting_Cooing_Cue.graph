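# Sound cue for one-shot ambient pigeon cooing at Asteria bridge POI spots, using bird one-shot attenuation and the ambient one-shot sound class.
# A top random node picks equally among strong, medium and soft sets, with mixers lowering strong and boosting medium.
# Each set randomly picks a wave or an empty slot, giving occasional silence.
# A modulator feeds the first two quality-level inputs; the other two are empty.
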SoundCue /Game/Sounds/Asteria/POI/Bridges/Spots/Spot_Bird_Pigeon_Roosting_Cooing_Cue
root=QualityLevel_0
nodes [
 {id=Mixer_0, children [Random_4], properties {inputVolume [0.5]}}
 {id=Mixer_1, children [Random_0], properties {inputVolume [1.2]}}
 {id=Modulator_0, children [Random_6]}
 {id=QualityLevel_0, children [Modulator_0,Modulator_0,null,null]}
 {
  id=Random_0
  children [WavePlayer_0,WavePlayer_1,WavePlayer_2,WavePlayer_3,WavePlayer_4,WavePlayer_5,null,null]
  properties {weights [1,1,1,1,1,1,1,1]}
 }
 {
  id=Random_1
  children [WavePlayer_6, WavePlayer_7, WavePlayer_8, WavePlayer_17, WavePlayer_18, WavePlayer_19, null, null]
  properties {weights [1,1,1,1,1,1,1,1]}
 }
 {
  id=Random_4
  children [WavePlayer_20, WavePlayer_21, WavePlayer_22, WavePlayer_23, WavePlayer_24, WavePlayer_25, WavePlayer_26, null, null, null]
  properties {weights [1,1,1,1,1,1,1,1,1,1]}
 }
 {id=Random_6, children [Mixer_0,Mixer_1,Random_1], properties {weights [1,1,1]}}
 {id=WavePlayer_0, properties {soundWaveAssetPtr=&/Spot_Bird_Pigeon_Cooing_Medium_01}}
 {id=WavePlayer_1, properties {soundWaveAssetPtr=&/Spot_Bird_Pigeon_Cooing_Medium_02}}
 {id=WavePlayer_17, properties {soundWaveAssetPtr=&/Spot_Bird_Pigeon_Cooing_Soft_04}}
 {id=WavePlayer_18, properties {soundWaveAssetPtr=&/Spot_Bird_Pigeon_Cooing_Soft_05}}
 {id=WavePlayer_19, properties {soundWaveAssetPtr=&/Spot_Bird_Pigeon_Cooing_Soft_06}}
 {id=WavePlayer_2, properties {soundWaveAssetPtr=&/Spot_Bird_Pigeon_Cooing_Medium_03}}
 {id=WavePlayer_20, properties {soundWaveAssetPtr=&/Spot_Bird_Pigeon_Cooing_Strong_01}}
 {id=WavePlayer_21, properties {soundWaveAssetPtr=&/Spot_Bird_Pigeon_Cooing_Strong_02}}
 {id=WavePlayer_22, properties {soundWaveAssetPtr=&/Spot_Bird_Pigeon_Cooing_Strong_03}}
 {id=WavePlayer_23, properties {soundWaveAssetPtr=&/Spot_Bird_Pigeon_Cooing_Strong_04}}
 {id=WavePlayer_24, properties {soundWaveAssetPtr=&/Spot_Bird_Pigeon_Cooing_Strong_05}}
 {id=WavePlayer_25, properties {soundWaveAssetPtr=&/Spot_Bird_Pigeon_Cooing_Strong_06}}
 {id=WavePlayer_26, properties {soundWaveAssetPtr=&/Spot_Bird_Pigeon_Cooing_Strong_07}}
 {id=WavePlayer_3, properties {soundWaveAssetPtr=&/Spot_Bird_Pigeon_Cooing_Medium_04}}
 {id=WavePlayer_4, properties {soundWaveAssetPtr=&/Spot_Bird_Pigeon_Cooing_Medium_05}}
 {id=WavePlayer_5, properties {soundWaveAssetPtr=&/Spot_Bird_Pigeon_Cooing_Medium_06}}
 {id=WavePlayer_6, properties {soundWaveAssetPtr=&/Spot_Bird_Pigeon_Cooing_Soft_01}}
 {id=WavePlayer_7, properties {soundWaveAssetPtr=&/Spot_Bird_Pigeon_Cooing_Soft_02}}
 {id=WavePlayer_8, properties {soundWaveAssetPtr=&/Spot_Bird_Pigeon_Cooing_Soft_03}}
]
properties {
 attenuationSettings=&Fort_TileAmbient_OneShot_Bird_Att
 bEnableBusSends=false
 maxDistance=8005
 soundClassObject=&Fort_Ambient_OneShots
 volumeMultiplier=0.601
 preEffectBusSends [{}]
}
