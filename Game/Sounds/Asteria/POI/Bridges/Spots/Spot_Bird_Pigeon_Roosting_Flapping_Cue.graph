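# Sound cue for one-shot pigeon roosting and wing-flap ambience, filed under the Asteria Bridges point-of-interest spot sounds.
# A quality-level node routes its first two levels to one shared random node and leaves the other two empty.
# The random node picks among seven equally weighted flapping waves.
# Playback uses bird one-shot attenuation and the ambient one-shots sound class, with a send to the exterior ambient source bus.
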SoundCue /Game/Sounds/Asteria/POI/Bridges/Spots/Spot_Bird_Pigeon_Roosting_Flapping_Cue
root=QualityLevel_0
nodes [
 {id=QualityLevel_0, children [Random_2,Random_2,null,null]}
 {
  id=Random_2
  children [WavePlayer_9, WavePlayer_10, WavePlayer_11, WavePlayer_12, WavePlayer_13, WavePlayer_14, WavePlayer_15]
  properties {weights [1,1,1,1,1,1,1]}
 }
 {id=WavePlayer_10, properties {soundWaveAssetPtr=&/Spot_Bird_Pigeon_Roosting_Flapping_02}}
 {id=WavePlayer_11, properties {soundWaveAssetPtr=&/Spot_Bird_Pigeon_Roosting_Flapping_03}}
 {id=WavePlayer_12, properties {soundWaveAssetPtr=&/Spot_Bird_Pigeon_Roosting_Flapping_04}}
 {id=WavePlayer_13, properties {soundWaveAssetPtr=&/Spot_Bird_Pigeon_Roosting_Flapping_05}}
 {id=WavePlayer_14, properties {soundWaveAssetPtr=&/Spot_Bird_Pigeon_Roosting_Flapping_06}}
 {id=WavePlayer_15, properties {soundWaveAssetPtr=&/Spot_Bird_Pigeon_Roosting_Flapping_07}}
 {id=WavePlayer_9, properties {soundWaveAssetPtr=&/Spot_Bird_Pigeon_Roosting_Flapping_01}}
]
properties {
 attenuationSettings=&Fort_TileAmbient_OneShot_Bird_Att
 maxDistance=8005
 soundClassObject=&Fort_Ambient_OneShots
 volumeMultiplier=0.752
 busSends [{audioBus=null, minSendLevel=0.6, sendLevel=1, soundSourceBus=&AmbientAudio_Exterior_SourceBus}]
}
